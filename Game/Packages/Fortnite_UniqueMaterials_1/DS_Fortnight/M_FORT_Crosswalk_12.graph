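# Nanite-enabled street material that paints crosswalk markings over a tiling asphalt road, built on the deprecated Fortress base shader function.
# Road diffuse, mask and normal textures project from world XY, so they tile independently of mesh UVs.
# The mesh-UV crosswalk mask's red channel blends in the squared Line color, and a world-projected blue channel wears it back to asphalt.
# The red mask also drives lerps of metalness toward zero and roughness toward a fixed value.
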Material /Game/Packages/Fortnite_UniqueMaterials_1/DS_Fortnight/M_FORT_Crosswalk_12
graph {
 properties {bUsedWithNanite=true, bUsedWithStaticLighting=true, pixelDepthOffsetMode=PDOM_Legacy, usedShadingModels=Default Lit}
 nodes [
  n1 = VectorParameter() {group=Code, parameterName=DiffuseColoration, defaultValue=<rgba 1 1 1 0>}
  AppendVector(A=@n1.RGB, B=@n1.RGB)
  Constant3Vector() {constant=<rgba 1 0 1 0>}
  n2 = WorldPosition()
  n3 = ComponentMask(None=Divide(A=@n2.XYZ, B=Constant() {r=-1024})) {g=true, r=true}
  n4 = TextureSample(UVs=@n3) {texture=/Game/Packages/DS_Fortnight/SM/Textures/T_Street_Road_D}
  n5 = VectorParameter() {parameterName=Line, defaultValue=<rgba 1.25 1.2 1.2 0.2>}
  n6 = Multiply(A=@n5.RGB, B=@n5.RGB)
  n7 = TextureSample() {texture=/Game/Packages/DS_Fortnight/T_FRT_Conc_Masks_Crosswalks}
  n8 = Constant() {r=2048}
  n9 = TextureSample(UVs=ComponentMask(None=Divide(A=@n2.XYZ, B=@n8)) {g=true, r=true}) {texture=&T_FRT_Conc_Masks_Crosswalks}
  n10 = MaterialFunctionCall!(
   -----------Base Shader-----------
   BaseColor1=LinearInterpolate(A=LinearInterpolate(A=@n4.RGB, B=@n6, Alpha=@n7.R), B=@n4.RGB, Alpha=@n9.B)
   Base Color Alpha
   SpecularMasks1=TextureSample(UVs=@n3) {samplerType=SAMPLERTYPE_Masks, texture=/Game/Packages/DS_Fortnight/SM/Textures/T_Street_Road_M}.RGB
   Emissive
   Normals1=TextureSample(UVs=@n3) {samplerType=SAMPLERTYPE_Normal, texture=/Game/Packages/DS_Fortnight/SM/Textures/T_Street_Road_N}.RGB
   World Position Offset
   DiffuseFresnelAmount=ScalarParameter() {defaultValue=0.25, parameterName=DiffuseFresnelAmount}
   -----------Destruction------------
   ------Multiple Diffuse Textures-----
   BaseColor2
   BaseColor3
   BaseColor4
   BaseColor5
   BaseColor6
   BaseColor7
   BaseColor8
   " - - - - - - - -"
   SpecularMasks2
   SpecularMasks3
   SpecularMasks4
   SpecularMasks5
   SpecularMasks6
   SpecularMasks7
   SpecularMasks8
   " - - - - - - - - "
   normal2
   Normal3
   Normal4
   Normal5
   Normal6
   Normal7
   Normal8
   " - - - - - - - - 1"
   Use 2 Textures
   Use 3 Textures
   Use 4 Textures
   Use 5 Textures
   Use 6 Textures
   Use 7 Textures
   Use 8 Textures
   Use 9 Textures
   Use 10 Textures
   ExposeLerpFunctionForInvertingTheHitDistanceCalc
  ) {materialFunction=/Game/Packages/Fortress_SharedMaterials/Base_Material_Functions/DEPRECATED_BaseShaderFunction}
  LinearInterpolate(A=@n10.Metalness, B=Constant(), Alpha=TextureSample() {texture=&T_FRT_Conc_Masks_Crosswalks}.R)
  LinearInterpolate(A=@n10.Roughness, B=Constant() {r=0.7}, Alpha=TextureSample() {texture=&T_FRT_Conc_Masks_Crosswalks}.R)
  MakeMaterialAttributes!(
   BaseColor
   Metallic
   Specular
   Roughness
   Anisotropy
   EmissiveColor
   Opacity
   OpacityMask
   Normal
   Tangent
   WorldPositionOffset
   SubsurfaceColor
   ClearCoat
   ClearCoatRoughness
   AmbientOcclusion
   Refraction
   CustomizedUVs_0
   CustomizedUVs_1
   CustomizedUVs_2
   CustomizedUVs_3
   CustomizedUVs_4
   CustomizedUVs_5
   CustomizedUVs_6
   CustomizedUVs_7
   PixelDepthOffset
   ShadingModel
   Displacement
  )
  MaterialFunctionCall(
   TextureObject=TextureObject() {texture=&T_Street_Road_D}
   TextureSize=Constant3Vector() {constant=<rgba 1024 1024 1024 0>}
  ) {materialFunction=&WorldAlignedTexture}
  Multiply(A=Multiply(A=@n7.R, B=OneMinus(None=@n9.B)), B=Constant() {r=1})
  ScalarParameter() {group=Code, parameterName=Code_DestructionAmount}
  ScalarParameter() {group=Code, parameterName=WSDamageRadius}
  ScalarParameter() {group=Code, parameterName=Code_HitAnimation}
  StaticBool() {value=true}
  VectorParameter() {parameterName=Asphalt, defaultValue=<rgba 0.0718 0.0738 0.0738 0>}
  VectorParameter() {parameterName=Code_DeformationVector}
  VectorParameter() {parameterName=Code_DeformationCenter}
  WorldPosition()
 ]
}
