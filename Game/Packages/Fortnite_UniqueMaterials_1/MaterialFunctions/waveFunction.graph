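# Material function that builds animated wave or round-bump World Position Offset on a panning, staggered world-space grid.
# World XY scrolls along Flow Direction, with alternate columns shifted half a cell.
# Sphere-mask shapes, tilted along the flow in wave mode, pulse on sines phase-offset by quantized cell position.
# Also outputs grid UVs, height gradient, edge falloff, and texture-driven cloud height, weighted normal and flow-facing tail mask.
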MaterialFunction /Game/Packages/Fortnite_UniqueMaterials_1/MaterialFunctions/waveFunction
graph {nodes [
  n1 = FunctionInput() {bUsePreviewValueAsDefault=true, inputName=Flow Direction, inputType=FunctionInput_Vector2, previewValue=<0 1 0 1>}
  n2 = FunctionInput() {bUsePreviewValueAsDefault=true, inputName=Time offset, inputType=FunctionInput_Scalar, previewValue=<0 1 0 1>}
  n3 = FunctionInput() {bUsePreviewValueAsDefault=true, inputName=Panner Time Multiplier, inputType=FunctionInput_Scalar, previewValue=<1 1 0 1>}
  n4 = Multiply(A=@n3, B=Time())
  n5 = WorldPosition()
  n6 = FunctionInput() {bUsePreviewValueAsDefault=true, inputName=Grid Size, inputType=FunctionInput_Vector2, previewValue=<128 128 0 1>}
  n7 = Add(A=Multiply(A=Add(A=@n2, B=@n4), B=@n1), B=Divide(A=ComponentMask(None=@n5.XYZ) {g=true, r=true}, B=@n6) {constB=64, desc=peak width}) {constA=0.5}
  n8 = ComponentMask(None=@n7) {r=true}
  n9 = Ceil(None=Add(A=@n8) {constB=0.5})
  n10 = ComponentMask(None=@n7) {g=true}
  n11 = AppendVector(A=@n8, B=Add(A=Multiply(A=@n9) {constB=0.5}, B=Multiply(A=@n10)))
  n12 = Frac(None=@n11)
  n13 = OneMinus(None=Abs(None=ConstantBiasScale(None=@n12) {bias=-0.5, scale=2}))
  n14 = ComponentMask(None=@n13) {r=true}
  n15 = ComponentMask(None=@n13) {g=true}
  n16 = Constant() {r=0.5}
  n17 = Multiply(
   A=OneMinus(None=If(A=@n14, B=@n15, "A > B"=@n14, "A < B"=@n15))
   B=OneMinus(None=SphereMask(A=@n12, B=@n16) {attenuationRadius=0.5, hardnessPercent=0})
  )
  n18 = Multiply(A=@n17, B=@n17)
  n19 = Multiply(A=@n18, B=@n18)
  n20 = Multiply(A=AppendVector(A=@n1, B=Constant()), B=@n19)
  n21 = FunctionInput() {bUsePreviewValueAsDefault=true, inputName=Tilt, inputType=FunctionInput_Scalar, previewValue=<-60 0 50 1>}
  n22 = FunctionInput() {bUsePreviewValueAsDefault=true, inputName=Grid Noise Tile Rep, inputType=FunctionInput_Vector2, previewValue=<5 3 0 1>}
  n23 = Frac(None=Divide(A=Add(A=@n11) {constB=-0.5}, B=@n22) {constB=5}) {desc=mod}
  n24 = Divide(A=Floor(None=Multiply(A=@n23, B=@n22) {constB=10}), B=@n22) {desc=0-1 values that have the number of increments specified}
  n25 = Frac(None=Add(A=ComponentMask(None=@n24) {g=true}, B=ComponentMask(None=@n24) {r=true}))
  n26 = FunctionInput() {bUsePreviewValueAsDefault=true, inputName=SineAnimSpeed, inputType=FunctionInput_Scalar, previewValue=<0.3 3 0 1>}
  n27 = FunctionInput() {bUsePreviewValueAsDefault=true, inputName=SineAnimOffset, inputType=FunctionInput_Scalar, previewValue=<0 3 0 1>}
  n28 = Add(A=Multiply(A=Time(), B=@n26), B=@n27)
  n29 = ConstantBiasScale(None=Sine(None=Add(A=@n25, B=@n28) {desc=offset wave phase}))
  n30 = FunctionInput() {bUsePreviewValueAsDefault=true, inputName=Height, inputType=FunctionInput_Scalar, previewValue=<50 0 50 1>}
  n31 = AppendVector(A=Constant2Vector(), B=@n30)
  n32 = Add(A=Multiply(A=Multiply(A=@n20, B=@n21) {constB=-60}, B=@n29), B=Multiply(A=@n19, B=@n31))
  n33 = Frac(None=Add(A=@n16, B=@n12))
  n34 = Multiply(A=SphereMask(A=Constant() {r=0.5}, B=@n33, Hardness=Constant()) {attenuationRadius=0.5}, B=@n31)
  n35 = StaticBool() {value=true}
  FunctionOutput(None=StaticSwitch(
    True=Multiply(A=@n32, B=@n29)
    False=Multiply(A=@n34, B=@n29)
    Value=FunctionInput(Preview=@n35) {bUsePreviewValueAsDefault=true, inputName="Type Wave(T)Round(F)", inputType=FunctionInput_StaticBool}
   )) {outputName=WPO, sortPriority=1}
  n36 = ComponentMask(None=Subtract(A=WorldPosition().XYZ, B=ActorPositionWS())) {b=true}
  n37 = FunctionInput() {bUsePreviewValueAsDefault=true, inputName=GradientZHeightFalloff, inputType=FunctionInput_Scalar, previewValue=<0 1 0 1>}
  n38 = Clamp(None=Divide(A=Subtract(A=@n36, B=@n37), B=@n30) {constB=50})
  FunctionOutput(None=Multiply(A=@n38, B=@n38)) {outputName=White height surface}
  n39 = Add(A=@n11) {constB=0.5}
  FunctionOutput(None=@n39) {outputName=UVOutput, sortPriority=1}
  n40 = FunctionInput() {bUsePreviewValueAsDefault=true, inputName=EdgeMaskSize, inputType=FunctionInput_Scalar, previewValue=<20 0 0 1>}
  n41 = Multiply(A=@n40, B=OneMinus(None=@n33)) {constA=10}
  n42 = Multiply(A=@n40, B=@n33) {constA=10}
  n43 = Clamp(None=MaterialFunctionCall!(f2, f3, f4=AppendVector(A=@n41, B=@n42)) {materialFunction=/Game/Packages/Fortnite_UniqueMaterials_1/VectorComponentOps}.f4 Min)
  n44 = TextureSample(UVs=@n39, Tex=FunctionInput() {bUsePreviewValueAsDefault=true, inputName=HeightMap, inputType=FunctionInput_Texture2D}) {samplerType=SAMPLERTYPE_Normal, texture=/Game/Effects/Fort_Effects/Textures/PointyNormal}
  FunctionOutput(None=Multiply(A=@n43, B=Multiply(A=@n44.R, B=@n29))) {outputName=CloudHeight, sortPriority=30}
  n45 = TextureSample(
   UVs=@n39
   Tex=FunctionInput(Preview=TextureObject() {samplerType=SAMPLERTYPE_Normal, texture=&PointyNormal}) {bUsePreviewValueAsDefault=true, inputName=DistortionNormal, inputType=FunctionInput_Texture2D}
  ) {samplerType=SAMPLERTYPE_Normal, texture=&PointyNormal}
  n46 = Multiply(A=@n44.R, B=@n38)
  FunctionOutput(None=Multiply(A=AppendVector(A=@n45.RGB, B=@n46), B=@n38)) {outputName=WeightedNormal, sortPriority=1}
  n47 = Clamp(None=DotProduct(A=@n1, B=ComponentMask(None=@n45.RGB) {g=true, r=true}))
  FunctionOutput(None=@n47) {outputName=Tail}
  FunctionOutput(None=@n43) {outputName=WPO_Falloff, sortPriority=1}
  Multiply(A=@n38, B=@n38)
  Multiply(A=@n38, B=@n47)
 ]}
