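# Translucent lit water surface material for Fortnite, with depth-faded edges and refraction and time-of-day control of color and roughness.
# Base color blends a tint with the scene behind, modulated by two panning world-projected caustic samples weighted by depth and surface normal.
# Four waveFunction layers drive vertex-color-masked displacement and normals, with UV-based sine ripples bending edge normals.
# Attributes route through the WaterRenderToTextureMaterialAtt function.
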Material /Game/Packages/Fortnite_UniqueMaterials_1/DS_Fortnight/M_FORT_Masked_Water_01
graph {
 properties {
  bScreenSpaceReflections=true
  bTangentSpaceNormal=false
  bUseMaterialAttributes=true
  blendMode=BLEND_Translucent
  physMaterial=/Game/FortressPhysicalMaterials/Water
  pixelDepthOffsetMode=PDOM_Legacy
  refractionMethod=RM_PixelNormalOffset
  translucencyLightingMode=TLM_SurfacePerPixelLighting
  translucencyPass=MTP_BeforeDOF
  usedShadingModels=Default Lit
 }
 nodes [
  n1 = ComponentMask(None=SceneTexture() {sceneTextureId=PPI_DiffuseColor}.Color) {b=true, g=true, r=true}
  n2 = Constant() {r=0.5}
  n3 = ComponentMask(None=LinearInterpolate(
    A=ShadingPathSwitch(Default=@n1, Deferred=@n1, Forward=@n2, Mobile=@n1)
    B=SceneColor!(UVs).RGB
    Alpha=ScalarParameter() {parameterName=Lighting}
   )) {b=true, g=true, r=true}
  n4 = Constant() {r=1}
  n5 = Divide(
   A=ComponentMask(None=MaterialFunctionCall() {materialFunction=&WorldPositionBehindTranslucency}.Result) {g=true, r=true}
   B=ScalarParameter() {defaultValue=512, parameterName=CausticScale}
  )
  n6 = Multiply(A=Time(), B=ScalarParameter() {defaultValue=3, parameterName=Speed})
  n7 = Panner(Coordinate=@n5, Time=@n6) {speedX=0.01, speedY=-0.003}
  n8 = Panner(Coordinate=Divide(A=@n5) {constB=1.618}, Time=@n6) {speedX=-0.003, speedY=-0.01}
  n9 = Add(A=LinearInterpolate(
    A=TextureSample(UVs=@n7) {samplerSource=SSM_Wrap_WorldGroupSettings, texture=/Game/Packages/Fortnite_UniqueMaterials_1/Textures/T_Water_Caustics_01}.G
    B=TextureSample(UVs=@n8) {samplerSource=SSM_Wrap_WorldGroupSettings, texture=&T_Water_Caustics_01}.G
   ))
  n10 = SceneTexture() {sceneTextureId=PPI_WorldNormal}
  n11 = Clamp(None=DotProduct(A=ComponentMask(None=@n10.Color) {b=true, g=true, r=true}, B=Constant3Vector() {constant=<rgba 0 0 1 0>}))
  n12 = ShadingPathSwitch(Default=@n11, Deferred=@n11, Forward=Constant() {r=0.5}, Mobile=@n11)
  n13 = VectorParameter() {parameterName=Modulate Color}
  n14 = Multiply(
   A=ShadingPathSwitch(Default=LinearInterpolate(B=@n9, Alpha=@n12) {constA=1}, Mobile=@n9) {desc=Migrated from FeatureLevelSwitch}
   B=Multiply(A=@n13.RGB, B=@n13.RGB)
  )
  n15 = Distance(
   A=ComponentMask(None=MaterialFunctionCall() {materialFunction=&WorldPositionBehindTranslucency}.Result) {b=true}
   B=ComponentMask(None=WorldPosition().XYZ) {b=true}
  )
  n16 = ScalarParameter() {defaultValue=32, parameterName=ModulateDistance}
  n17 = LinearInterpolate(B=Clamp(None=Divide(A=@n15, B=@n16)), Alpha=ScalarParameter() {parameterName=ModulateMin}) {constA=1}
  n18 = VectorParameter() {parameterName=Diffuse}
  n19 = LinearInterpolate(
   A=Multiply(A=ShadingPathSwitch(Default=@n3, Mobile=@n4) {desc=Migrated from FeatureLevelSwitch}, B=LinearInterpolate(B=@n14, Alpha=@n17) {constA=1}) {constA=1, desc=Modulate}
   B=Multiply(A=@n18.RGB, B=@n18.RGB)
   Alpha=ScalarParameter() {defaultValue=0.7, parameterName=Alpha}
  )
  n20 = CollectionParameter() {
   collection=&FortniteMaterialParameters
   parameterName=FortniteActiveTimeOfDay
  }
  n21 = VectorParameter() {group=Time of Day, parameterName=Base Color Values Across The Times Of Day, defaultValue=<rgba 1.5 0.5 2.5 2.5>}
  n22 = DotProduct(A=@n20, B=AppendVector(A=@n21.RGB, B=@n21.RGB))
  n23 = ScalarParameter() {group=Time of Day, parameterName=NightRoughness}
  n24 = ScalarParameter() {defaultValue=0.05, group=Time of Day, parameterName=DayRoughness}
  n25 = CollectionParameter() {
   collection=&FortniteMaterialParameters
   parameterName=FortniteActiveTimeOfDay
  }
  n26 = DotProduct(A=@n25, B=Constant4Vector() {constant=<rgba 1 1 1 0>}.RGBA)
  n27 = Constant() {r=4}
  n28 = VertexColor()
  n29 = Constant() {r=0.75}
  n30 = Constant() {r=3}
  n31 = Constant() {r=2.5}
  n32 = Constant() {r=0.2}
  n33 = Constant() {r=46}
  n34 = Constant()
  n35 = ScalarParameter() {defaultValue=-64.6579, parameterName=ZoffsetOfWaterPlaneFromPivotZLocation}
  n36 = TextureObject() {samplerType=SAMPLERTYPE_LinearColor, texture=/Game/Effects/Fort_Effects/Textures/HeightForWater}
  n37 = MaterialFunctionCall!(
   Panner Time Multiplier=Multiply(B=@n29) {constA=1}
   Grid Size=Multiply(A=@n30, B=Multiply(A=@n31) {constB=100})
   SineAnimSpeed=@n32
   Flow Direction=Constant2Vector() {r=0.4}
   Height=@n33
   Tilt=Multiply(A=@n34, B=@n31)
   Time offset=Constant() {r=0.33}
   Grid Noise Tile Rep
   GradientZHeightFalloff=@n35
   SineAnimOffset
   "Type Wave(T)Round(F)"
   DistortionNormal
   HeightMap=@n36
   EdgeMaskSize
  ) {materialFunction=/Game/Packages/Fortnite_UniqueMaterials_1/MaterialFunctions/waveFunction}
  n38 = Constant() {r=1.5}
  n39 = MaterialFunctionCall(
   Panner Time Multiplier=Multiply(B=@n29) {constA=1}
   Grid Size=Multiply(A=@n30, B=Multiply(A=@n38) {constB=100})
   SineAnimSpeed=@n32
   Flow Direction=Constant2Vector() {r=-0.5}
   Height=@n33
   Tilt=Multiply(A=@n34, B=@n38)
   Time offset=Constant() {r=0.5}
   GradientZHeightFalloff=@n35
   SineAnimOffset=Constant() {r=1}
   HeightMap=@n36
  ) {materialFunction=&waveFunction}
  n40 = Constant() {r=3.5}
  n41 = MaterialFunctionCall(
   Panner Time Multiplier=Multiply(B=@n29) {constA=1}
   Grid Size=Multiply(A=@n30, B=Multiply(A=@n40) {constB=100})
   SineAnimSpeed=@n32
   Flow Direction=Constant2Vector() {g=0.25, r=0.25}
   Height=@n33
   Tilt=Multiply(A=@n34, B=@n40)
   Time offset=Constant() {r=2.75}
   GradientZHeightFalloff=@n35
   SineAnimOffset=Constant() {r=2}
   HeightMap=@n36
  ) {materialFunction=&waveFunction}
  n42 = Constant() {r=1.28}
  n43 = MaterialFunctionCall(
   Panner Time Multiplier=Multiply(B=@n29) {constA=1, constB=0.5}
   Grid Size=Multiply(A=@n30, B=Multiply(A=@n42) {constB=100})
   SineAnimSpeed=@n32
   Flow Direction=Constant2Vector() {g=0.5}
   Height=@n33
   Tilt=Multiply(A=@n34, B=@n42)
   Time offset=Constant() {r=5}
   Grid Noise Tile Rep=Constant2Vector() {desc=4, g=4, r=4}
   GradientZHeightFalloff=@n35
   HeightMap=@n36
  ) {materialFunction=&waveFunction}
  n44 = Add(A=@n39.WPO, B=Add(A=@n41.WPO, B=@n43.WPO))
  n45 = OneMinus(None=StaticSwitchParameter(True=VertexColor().R, False=Constant() {r=1}) {defaultValue=true, group=optimizations, parameterName=Evaluate Deep Water})
  n46 = Multiply(A=Clamp(None=Add(A=@n37.WPO, B=@n44) {desc=WPO}, Max=@n33), B=Clamp(None=Multiply(A=@n45, B=@n45)))
  n47 = ScalarParameter() {defaultValue=1.33, parameterName=Refraction}
  n48 = DepthFade(FadeDistance=ScalarParameter() {defaultValue=8, parameterName=Distortion Depth Fade})
  n49 = Add(A=@n37.WeightedNormal, B=@n39.WeightedNormal)
  n50 = Add(A=@n41.WeightedNormal, B=@n43.WeightedNormal)
  n51 = Add(A=ComponentMask(None=Add(A=@n49, B=@n50)) {b=true, g=true, r=true}, B=Constant3Vector() {constant=<rgba 0 0 0.001 0>}) {constB=0.001}
  n52 = Normalize(VectorInput=Multiply(A=Normalize(VectorInput=@n51) {desc=adds 60 Instructions}, B=Constant3Vector() {constant=<rgba 0.2 0.2 1 0>})) {desc=adds 60 Instructions}
  n53 = Constant() {desc=Final Perma Angle, r=-2}
  n54 = Clamp(None=MaterialFunctionCall(Float2=TextureCoordinate()) {materialFunction=&BreakOutFloat2Components}.R)
  n55 = OneMinus(None=@n54)
  n56 = Subtract(A=@n55, B=Constant() {desc=Edge bend start point, r=0.7715})
  n57 = Clamp(None=Multiply(A=Clamp(None=@n56), B=Constant() {desc=Constant Edge Bend Amount, r=3.8372}))
  n58 = Constant() {desc=Ripple Falloff, r=0.9895}
  n59 = Multiply(A=@n54, B=ScalarParameter() {defaultValue=1.25, parameterName=Edge Ripple Size})
  n60 = Multiply(A=Time(), B=ScalarParameter() {defaultValue=0.25, group=Ripple, parameterName=Edge Ripple Speed}) {constB=0.25}
  n61 = Sine(None=Add(A=Multiply(A=@n59) {constB=-1}, B=@n60)) {period=0.5346}
  n62 = WorldPosition()
  n63 = Constant3Vector() {desc=WS Ripple Affect, constant=<rgba -0.0045 -0.0025 0 1>}
  n64 = Add(A=@n59, B=Multiply(A=DotProduct(A=@n62.XYZ, B=@n63)))
  n65 = Multiply(A=Sine(None=Add(A=@n60, B=@n64)) {period=0.7586}) {constB=0.5}
  n66 = Multiply(A=Multiply(A=Power(Base=@n55, Exp=@n58), B=Add(A=@n61, B=@n65)), B=Constant() {r=0.5413})
  n67 = RotateAboutAxis(
   NormalizedRotationAxis=Transform(None=Constant3Vector() {constant=<rgba 0 1 0 0>})
   RotationAngle=Multiply(A=LinearInterpolate(A=@n66, B=@n53, Alpha=@n57), B=ScalarParameter() {defaultValue=0.1688, parameterName=Ripple Amount})
   PivotPoint=Constant()
   Position=@n52
  )
  MaterialFunctionCall!(
   Render To Texture Pass
   WS Flow Map Dir=Constant2Vector()
   Opacity
   Default Mat Att=MakeMaterialAttributes!(
    BaseColor=Multiply(A=@n19, B=@n22)
    Metallic
    Specular=ScalarParameter() {defaultValue=0.05, parameterName=Water Spec}
    Roughness=LinearInterpolate(A=@n23, B=@n24, Alpha=@n26)
    Anisotropy
    EmissiveColor
    Opacity=DepthFade(FadeDistance=@n27)
    OpacityMask
    Normal=Add(A=@n52, B=@n67)
    Tangent
    WorldPositionOffset=Multiply(A=@n28.A, B=@n46) {desc=wpo}
    SubsurfaceColor
    ClearCoat
    ClearCoatRoughness
    AmbientOcclusion
    Refraction=LinearInterpolate(B=@n47, Alpha=@n48) {constA=1}
    CustomizedUVs_0
    CustomizedUVs_1
    CustomizedUVs_2
    CustomizedUVs_3
    CustomizedUVs_4
    CustomizedUVs_5
    CustomizedUVs_6
    CustomizedUVs_7
    PixelDepthOffset
    ShadingModel
    Displacement
   )
  ) {materialFunction=/Game/WaterAndWind/Materials/Water/WaterRenderToTextureMaterialAtt}
 ]
}
