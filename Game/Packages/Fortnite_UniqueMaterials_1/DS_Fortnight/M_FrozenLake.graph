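# Frozen lake surface material with cracked ice and snow, world-mapped by default.
# Crack depth is faked by offsetting extra crack samples along the view direction, with temporal dither between two heights.
# A noise mask mixes snow into the ice, which is then height-blended with a snow layer using vertex red.
# Switches add shared winter snow (default on) and a dithered opacity mask from snow texture and vertex alpha (default off).
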Material /Game/Packages/Fortnite_UniqueMaterials_1/DS_Fortnight/M_FrozenLake
graph {
 properties {bTangentSpaceNormal=false, bUseMaterialAttributes=true, pixelDepthOffsetMode=PDOM_Legacy, usedShadingModels=Default Lit}
 nodes [
  Constant()
  Constant()
  n1 = VectorParameter() {parameterName=Color_1, defaultValue=<rgba 0.0781 0.0781 0.0781 1>}
  n2 = StaticSwitchParameter(True=ComponentMask(None=WorldPosition().XYZ) {g=true, r=true}, False=TextureCoordinate()) {defaultValue=true, parameterName=UseWorldMapping}
  n3 = TextureSample(UVs=Divide(A=@n2, B=ScalarParameter() {defaultValue=243.8069, parameterName=Size_2})) {
   samplerSource=SSM_Wrap_WorldGroupSettings
   samplerType=SAMPLERTYPE_Masks
   texture=/Game/Packages/Fortnite_UniqueMaterials_1/Textures/T_Ice_Lake_Cracks_M
  }
  LinearInterpolate(A=@n1.RGB, B=VectorParameter() {parameterName=Color_2, defaultValue=<rgba 0.0573 0.0573 0.0573 1>}.RGB, Alpha=@n3.G)
  ScalarParameter() {parameterName=Roughness}
  ScalarParameter() {defaultValue=0.005, parameterName=Ratio}
  StaticBoolParameter() {parameterName=UseVertexBlend}
  n4 = TextureSample(UVs=Divide(A=@n2, B=ScalarParameter() {defaultValue=408.0741, parameterName=Size_3})) {samplerSource=SSM_Wrap_WorldGroupSettings, texture=/Game/Packages/Fortnite_UniqueMaterials_1/Textures/T_Ice_Lake_Snow_D}
  n5 = ScalarParameter() {defaultValue=0.7216, parameterName=Roughness_Snow_1}
  n6 = Divide(A=@n2, B=ScalarParameter() {defaultValue=654.517, parameterName=Size})
  n7 = TextureSample(UVs=@n6) {
   samplerSource=SSM_Wrap_WorldGroupSettings
   samplerType=SAMPLERTYPE_Masks
   texture=&T_Ice_Lake_Cracks_M
  }
  n8 = LinearInterpolate(
   A=ScalarParameter() {defaultValue=-8.0721, parameterName=Contrast SharpCracks}
   B=ScalarParameter() {defaultValue=24.2048, parameterName=Contrast SharpCracks_1}
   Alpha=@n7.R
  )
  n9 = CameraVectorWS()
  n10 = Multiply(A=@n9, B=Constant3Vector() {constant=<rgba 1 1 0 0>})
  n11 = Constant3Vector() {constant=<rgba 1 1 0 0>}
  n12 = ScalarParameter() {parameterName=Height_1}
  n13 = ScalarParameter() {defaultValue=-0.007, parameterName=Height_2}
  n14 = MaterialFunctionCall() {materialFunction=&DitherTemporalAA}
  n15 = VertexNormalWS()
  n16 = Multiply(
   A=ComponentMask(None=MaterialFunctionCall(Vector=@n10, Default=@n11) {materialFunction=&SafeNormalize}.Result) {g=true, r=true}
   B=Divide(A=LinearInterpolate(A=@n12, B=@n13, Alpha=@n14.Result), B=DotProduct(A=@n15, B=@n9))
  )
  n17 = TextureSample(UVs=Add(A=@n6, B=@n16)) {
   samplerSource=SSM_Wrap_WorldGroupSettings
   samplerType=SAMPLERTYPE_Masks
   texture=&T_Ice_Lake_Cracks_M
  }
  n18 = ScalarParameter() {defaultValue=1.5, parameterName=Offset}
  n19 = TextureSample(UVs=Add(A=@n6, B=Multiply(A=@n16, B=@n18))) {
   samplerSource=SSM_Wrap_WorldGroupSettings
   samplerType=SAMPLERTYPE_Masks
   texture=&T_Ice_Lake_Cracks_M
  }
  n20 = Add(
   A=Add(A=@n7.R, B=Clamp(None=@n8) {clampMode=CMODE_ClampMin})
   B=Multiply(A=Add(A=@n17.R, B=@n19.R), B=ScalarParameter() {defaultValue=2.2005, parameterName=DepthBrightness})
  )
  n21 = LinearInterpolate(
   A=@n1.RGB
   B=VectorParameter() {parameterName=Color_3, defaultValue=<rgba 0.2656 0.2656 0.2656 1>}.RGB
   Alpha=Clamp(None=Divide(A=@n20) {constB=3})
  )
  n22 = TextureSample(UVs=Divide(A=@n2, B=ScalarParameter() {defaultValue=706.7794, parameterName=Size_Blendmask})) {
   samplerSource=SSM_Wrap_WorldGroupSettings
   samplerType=SAMPLERTYPE_LinearColor
   texture=/Game/Packages/Fortnite_UniqueMaterials_1/Textures/water2_DISP
  }
  n23 = LinearInterpolate(
   A=ScalarParameter() {defaultValue=-0.2601, parameterName=Transition_1}
   B=ScalarParameter() {defaultValue=0.6438, parameterName=Transition_2}
   Alpha=@n22.R
  )
  n24 = LinearInterpolate(
   A=LinearInterpolate(
    A=ScalarParameter() {defaultValue=0.1061, parameterName=Roughness_1}
    B=ScalarParameter() {defaultValue=0.2176, parameterName=Roughness_2}
    Alpha=@n3.G
   )
   B=ScalarParameter() {defaultValue=0.2952, parameterName=Roughness_3}
   Alpha=@n20
  )
  n25 = ScalarParameter() {defaultValue=0.7216, parameterName=Roughness_Snow}
  n26 = MaterialFunctionCall!(
   Base Material=MakeMaterialAttributes!(
    BaseColor=@n4.RGB
    Metallic
    Specular
    Roughness=Multiply(A=@n5, B=@n4.R)
    Anisotropy
    EmissiveColor
    Opacity
    OpacityMask
    Normal
    Tangent
    WorldPositionOffset
    SubsurfaceColor
    ClearCoat
    ClearCoatRoughness
    AmbientOcclusion
    Refraction
    CustomizedUVs_0
    CustomizedUVs_1
    CustomizedUVs_2
    CustomizedUVs_3
    CustomizedUVs_4
    CustomizedUVs_5
    CustomizedUVs_6
    CustomizedUVs_7
    PixelDepthOffset
    ShadingModel
    Displacement
   )
   Top Material=MakeMaterialAttributes(
    BaseColor=LinearInterpolate(A=@n21, B=@n4.RGB, Alpha=@n23)
    Specular=Constant() {r=1}
    Roughness=LinearInterpolate(A=@n24, B=@n25, Alpha=@n23)
   )
   Alpha=MaterialFunctionCall(Transition Phase=VertexColor().R, Height Texture=@n4.R, Contrast=ScalarParameter() {defaultValue=0.0145, parameterName=ContrastSnowBlend}) {materialFunction=&HeightLerp}.Alpha
  ) {materialFunction=/Engine/Functions/MaterialLayerFunctions/MatLayerBlend_Standard}
  n27 = StaticSwitchParameter(
   True=MaterialFunctionCall!(MaterialLayers=@n26.Blended Material, UseVertexBlend=StaticBool() {value=true}) {materialFunction=/Game/Environments/Sets/Winter/Terrain/MF_SharedSnow}.OutputBlendedMaterial
   False=@n26.Blended Material
  ) {defaultValue=true, parameterName=UseSnowBlend}
  n28 = LinearInterpolate(
   A=ScalarParameter() {defaultValue=-0.8, parameterName=ContrastOpacityMask_1}
   B=ScalarParameter() {defaultValue=1.4229, parameterName=ContrastOpacityMask_2}
   Alpha=Add(A=@n4.R, B=VertexColor().A)
  )
  StaticSwitchParameter(
   True=MaterialFunctionCall!(
    Material=@n27
    NewOpacityMask=MaterialFunctionCall(Alpha Threshold=@n28) {materialFunction=&DitherTemporalAA}.Result
    Mask
   ) {materialFunction=/Engine/Functions/MaterialLayerFunctions/MatLayerBlend_OverrideOpacityMask}.Blended Material
   False=@n27
  ) {parameterName=isMasked}
 ]
}
